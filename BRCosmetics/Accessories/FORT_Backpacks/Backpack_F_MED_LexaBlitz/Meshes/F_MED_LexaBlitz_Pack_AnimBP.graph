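# Animation Blueprint for the LexaBlitz backpack, adding procedural motion to the pack and its pieces.
# Spine bend yaw drives additive rotation and offset on the pack bone, while time-based sine waves oscillate the small piece bones.
# Five chained AnimDynamics nodes then add linear-spring motion to the main pieces.
# Gravity override (from FlintTwirl's parameter asset) and "disable dynamics" curve alpha are computed but not wired into those nodes.
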AnimBlueprint /BRCosmetics/Accessories/FORT_Backpacks/Backpack_F_MED_LexaBlitz/Meshes/F_MED_LexaBlitz_Pack_AnimBP
target_skeleton=&/F_MED_LexaBlitz_Pack_Skeleton
graphs [
 AnimGraph {nodes [
   n1 = "Transform (Modify) Bone - Bone: pack"(ComponentPose=Local To Component(LocalPose=Input Pose()), Translation=Get backpack_translation(), Rotation=Get backpack_rotate()) {rotationMode=BMM_Additive, translationMode=BMM_Additive, alphaBoolBlend {blendOption=Linear}, boneToModify {boneName=pack}}
   n2 = Get Pieces move() -> (Pieces move:Vector)
   n3 = "Transform (Modify) Bone - Bone: pieces_1_l"!(ComponentPose=@n1, Alpha=0.8, Translation=@n2) -> (Pose) {translationMode=BMM_Additive, translationSpace=BCS_BoneSpace, alphaBoolBlend {blendOption=Linear}, boneToModify {boneName=pieces_1_l}}
   n4 = "Transform (Modify) Bone - Bone: pieces_1_r"!(ComponentPose=@n3, Alpha=0.8, Translation=@n2) -> (Pose) {translationMode=BMM_Additive, translationSpace=BCS_BoneSpace, alphaBoolBlend {blendOption=Linear}, boneToModify {boneName=pieces_1_r}}
   n5 = Get Pieces move 2() -> (Pieces move 2:Vector)
   n6 = "Transform (Modify) Bone - Bone: pieces_2_l"!(ComponentPose=@n4, Alpha=0.8, Translation=@n5) -> (Pose) {translationMode=BMM_Additive, translationSpace=BCS_BoneSpace, alphaBoolBlend {blendOption=Linear}, boneToModify {boneName=pieces_2_l}}
   n7 = "Transform (Modify) Bone - Bone: pieces_2_r"!(ComponentPose=@n6, Alpha=0.8, Translation=@n5) -> (Pose) {translationMode=BMM_Additive, translationSpace=BCS_BoneSpace, alphaBoolBlend {blendOption=Linear}, boneToModify {boneName=pieces_2_r}}
   n8 = "Transform (Modify) Bone - Bone: pieces_m"!(ComponentPose=@n7, Alpha=0.8, Translation=Get Pieces move 3() -> (Pieces move 3:Vector)) -> (Pose) {translationMode=BMM_Additive, translationSpace=BCS_BoneSpace, alphaBoolBlend {blendOption=Linear}, boneToModify {boneName=pieces_m}}
   n9 = Anim Dynamics - Bone: pieces_main_1_l!(SimulationSpace:AnimPhysSimSpaceType Enum, ComponentPose=@n8, Alpha=1, GravityOverride:Vector="0,0,0", ExternalForce="0,0,0") -> (Pose) {
    angularDampingOverride=1
    bGravityOverrideInSimSpace=true
    bLinearSpring=true
    bOverrideLinearDamping=true
    bUseGravityOverride=true
    linearDampingOverride=1
    linearSpringConstant=61.7856
    alphaBoolBlend {blendOption=Linear}
    boundBone {boneName=pieces_main_1_l}
    componentLinearAccScale {z=0.02}
    componentLinearVelScale {z=0.03}
    physicsBodyDefinitions [{
      boundBone {boneName=pieces_main_1_l}
      boxExtents=<20 20 20>
      constraintSetup {
       angularTargetAxis=AxisY
       coneAngle=1
       angularLimitsMax=<0 0 0>
       angularLimitsMin=<0 0 0>
       angularTarget=<1 0 0>
       linearAxesMax=<2 0 0>
       linearAxesMin=<-5 0 0>
      }
      localJointOffset=<0 0 0>
     }]
   }
   n10 = Anim Dynamics - Bone: pieces_main_2_l!(SimulationSpace:AnimPhysSimSpaceType Enum, ComponentPose=@n9, Alpha=1, GravityOverride:Vector="0,0,0", ExternalForce="0,0,0") -> (Pose) {
    angularDampingOverride=1
    bGravityOverrideInSimSpace=true
    bLinearSpring=true
    bOverrideLinearDamping=true
    bUseGravityOverride=true
    linearDampingOverride=1
    linearSpringConstant=44.4181
    alphaBoolBlend {blendOption=Linear}
    boundBone {boneName=pieces_main_2_l}
    componentLinearAccScale {z=0.02}
    componentLinearVelScale {z=0.03}
    physicsBodyDefinitions [{
      boundBone {boneName=pieces_main_2_l}
      boxExtents=<20 20 20>
      constraintSetup {
       angularTargetAxis=AxisY
       coneAngle=1
       angularLimitsMax=<0 0 0>
       angularLimitsMin=<0 0 0>
       angularTarget=<1 0 0>
       linearAxesMax=<2 0 0>
       linearAxesMin=<-5 0 0>
      }
      localJointOffset=<0 0 0>
     }]
   }
   n11 = Anim Dynamics - Bone: pieces_main_1_r!(SimulationSpace:AnimPhysSimSpaceType Enum, ComponentPose=@n10, Alpha=1, GravityOverride:Vector="0,0,0", ExternalForce="0,0,0") -> (Pose) {
    angularDampingOverride=1
    bGravityOverrideInSimSpace=true
    bLinearSpring=true
    bOverrideLinearDamping=true
    bUseGravityOverride=true
    linearDampingOverride=1
    linearSpringConstant=74.1914
    alphaBoolBlend {blendOption=Linear}
    boundBone {boneName=pieces_main_1_r}
    componentLinearAccScale {y=0.03, z=0.02}
    componentLinearVelScale {y=0.02, z=0.03}
    physicsBodyDefinitions [{
      boundBone {boneName=pieces_main_1_r}
      boxExtents=<20 20 20>
      constraintSetup {coneAngle=1, angularLimitsMax=<0 0 0>, angularLimitsMin=<0 0 0>, angularTarget=<1 0 0>, linearAxesMax=<2 0 0>, linearAxesMin=<-5 0 0>}
      localJointOffset=<0 0 0>
     }]
   }
   n12 = Anim Dynamics - Bone: pieces_main_2_r!(SimulationSpace:AnimPhysSimSpaceType Enum, ComponentPose=@n11, Alpha=1, GravityOverride:Vector="0,0,0", ExternalForce="0,0,0") -> (Pose) {
    angularDampingOverride=1
    bGravityOverrideInSimSpace=true
    bLinearSpring=true
    bOverrideLinearDamping=true
    bUseGravityOverride=true
    linearDampingOverride=1
    linearSpringConstant=19.2738
    alphaBoolBlend {blendOption=Linear}
    boundBone {boneName=pieces_main_2_r}
    componentLinearAccScale {z=0.03}
    componentLinearVelScale {z=0.035}
    physicsBodyDefinitions [{
      boundBone {boneName=pieces_main_2_r}
      boxExtents=<20 20 20>
      constraintSetup {
       angularTargetAxis=AxisY
       coneAngle=1
       angularLimitsMax=<0 0 0>
       angularLimitsMin=<0 0 0>
       angularTarget=<1 0 0>
       linearAxesMax=<2 0 0>
       linearAxesMin=<-5 0 0>
      }
      localJointOffset=<0 0 0>
     }]
   }
   n13 = Anim Dynamics - Bone: pieces_main_m!(SimulationSpace:AnimPhysSimSpaceType Enum, ComponentPose=@n12, Alpha=1, ExternalForce="0,0,0") -> (Pose) {
    angularDampingOverride=1
    bAngularSpring=true
    bLinearSpring=true
    bOverrideLinearDamping=true
    bUseGravityOverride=true
    linearDampingOverride=0.9999
    linearSpringConstant=23.4685
    alphaBoolBlend {blendOption=Linear}
    boundBone {boneName=pieces_main_m}
    componentLinearAccScale {z=0.02}
    componentLinearVelScale {z=0.03}
    physicsBodyDefinitions [{
      boundBone {boneName=pieces_main_m}
      boxExtents=<25 25 25>
      constraintSetup {
       angularTargetAxis=AxisY
       coneAngle=1
       angularLimitsMax=<0 0 0>
       angularLimitsMin=<0 0 0>
       angularTarget=<1 0 0>
       linearAxesMax=<2 0 0>
       linearAxesMin=<-4 0 0>
      }
      localJointOffset=<0 0 0>
     }]
   }
   Output Pose(Result=Component To Local(ComponentPose=@n13)) {layerGroup=None}
   Reroute Node(InputPin:Vector=Get GravityOverride()) -> (OutputPin:Vector)
  ]}
 EventGraph {nodes [
   Event BlueprintUpdateAnimation()
   > n1 = Sequence!(execute) -> (then_0, then_1, then_2:Exec)
   n2 = "Gravity Override State Function (Parameter Asset) (Nativized)"(
    execute=@n1.then_2
    CharacterPartAnimInstance=Self-Reference()
    GameTimeInSeconds=GetGameTimeInSeconds()
    Parameters=/BRCosmetics/Accessories/FORT_Backpacks/Backpack_F_MED_FlintTwirl/Meshes/M_MED_FlintTwirl_Parameters
   )
   > Set GravityOverride(GravityOverride=@n2)
   > Set RBAN Settings Out_Native(RBAN Settings Out_Native:Fort Rigid Body Anim Node Parameters Structure=@n2.OutRigidBodyAnimNodeParameters) -> (Output_Get:Fort Rigid Body Anim Node Parameters Structure)
   Set curve_alpha(
    execute=@n1.then_1
    curve_alpha:float=MapRangeClamped(Value=GetCurveValue(CurveName=disable dynamics), InRangeA=0, InRangeB=1, OutRangeA=1, OutRangeB=0)
   ) -> (Output_Get:float)
   n3 = GetRealTimeSeconds()
   n4 = "Sin (Radians)"(A=float + float(A=float * float(A=@n3, B=1.6) {is=title K2Node_PromotableOperator}))
   > Set Pieces move(Pieces move:Vector=MakeVector(X=float * float(A=@n4, B=1), Y=0)) -> (Output_Get:Vector)
   n5 = "Sin (Radians)"(A=float + float(A=float * float(A=@n3, B=2)))
   > Set Pieces move 2(Pieces move 2:Vector=MakeVector(X=float * float(A=@n5, B=0.8), Y=0)) -> (Output_Get:Vector)
   n6 = "Sin (Radians)"(A=float + float(A=float * float(A=@n3, B=2.4)))
   > Set Pieces move 3(Pieces move 3:Vector=MakeVector(X=float * float(A=@n6, B=0.8), Y=0)) -> (Output_Get:Vector)
   n7 = GetOwnerSkeletalMeshComponent()
   n8 = BreakRotator(InRot=BreakTransform(InTransform=GetDeltaTransformFromRefPose(self=@n7, BoneName=spine_05, BaseName=spine_01)).Rotation)
   Set backpack_rotate(execute=@n1.then_0, backpack_rotate=MakeRotator(Roll=MapRangeClamped(Value=@n8.Yaw, InRangeA=-30, InRangeB=-150, OutRangeB=240)))
   > Set backpack_translation(backpack_translation=MakeVector(
     Y=MapRangeClamped(Value=@n8.Yaw, InRangeA=-20, InRangeB=-50, OutRangeB=-13)
     Z=MapRangeClamped(Value=@n8.Yaw, InRangeA=-30, InRangeB=-50, OutRangeB=-2)
    ))
   TryGetPawnOwner()
  ]}
]
variables [
 backpack_rotate {type=Rotator}
 backpack_translation {type=Vector}
 curve_alpha {type=number}
 GravityOverride {type=Vector}
 Pieces move {type=Vector}
 Pieces move 2 {type=Vector}
 Pieces move 3 {type=Vector}
 RBAN Settings Out_Native {type=FortRigidBodyAnimNodeParameters}
]
